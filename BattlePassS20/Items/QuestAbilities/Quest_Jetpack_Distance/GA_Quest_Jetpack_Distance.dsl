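; Battle pass quest gameplay ability that reports distance the owning player travels while it is active, named for jetpack use.
; Activation starts a looping timer at a randomized sub-second interval, adding straight-line movement since the last sample in floored meters.
; Totals above a threshold go to the player's quest manager as custom stat events, then reset.
; Ending clears the timer and sends leftover meters without resampling location.
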(event Ability|EventActivateAbilityFromEvent
  (params (EventData :type Gameplay%20Event%20Data%20Structure%20%28by%20ref%29))
  ;[1 1 1] Activate Ability, set up tracking vars
  (|CallParentFunction EventData)
  (bind _output_get (Variables|Default|SetPreviousLocation (Transformation|GetActorLocation (Variables|GAVars|Data|GetOwningPlayerPawn))))
  (Variables|Default|SetCurrentLocation _output_get)
  (bind _output_get_2 (Variables|Default|SetTickInterval (Math|Random|RandomFloatinRange 0.5 1.0)))
  (bind _returnvalue (Utilities|Time|SetTimerbyEvent (delegate Custom|SendDistanceUpdate) _output_get_2 true))
  (Variables|Default|SetTimerHandle _returnvalue))

(event Custom|SendDistanceUpdate
  ;[1 1 1] Calculate and send distance update
  (Utilities|FlowControl|Sequence
    (:then_0
      (Variables|Default|SetCurrentLocation (Transformation|GetActorLocation (Variables|GAVars|Data|GetOwningPlayerPawn)))
      (Variables|Default|SetMetersTravelled (+ (Math|Float|Floor (/ (Math|Vector|VectorLength (- (Variables|Default|GetCurrentLocation) (Variables|Default|GetPreviousLocation))) 1e+02)) (Variables|Default|GetMetersTravelled)))
      (Variables|Default|SetPreviousLocation (Variables|Default|GetCurrentLocation)))
    (:then_1
      (if (> (Variables|Default|GetMetersTravelled) 30)
        (Quest|SendComplexCustomStatEvent (Variables|GAVars|Data|GetOwningPlayerQuestManager) 0 (Variables|Default|GetT_Quest) "(GameplayTags=)" (Variables|Default|GetMetersTravelled))
        (Variables|Default|SetMetersTravelled 0)))))

(event Ability|EventOnEndAbility
  (params (bWasCancelled :type Boolean))
  (bind _timerhandle (Variables|Default|GetTImerHandle))
  ;[1 1 1] End Ability - clear timer, send a final distance update for any meters travelled since last update
  (|CallParentFunction bWasCancelled)
  (if (Utilities|Time|IsValidTimerHandle _timerhandle)
    (Utilities|Time|ClearandInvalidateTimerbyHandle _timerhandle)
    (else
      _))
  ;;[1 1 1] Check if player moved; if so send final update
  (if (>= (Variables|Default|GetMetersTravelled) 1)
    ;[1 1 1] Calculate and send distance update
    (Quest|SendComplexCustomStatEvent (Variables|GAVars|Data|GetOwningPlayerQuestManager) 0 (Variables|Default|GetT_Quest) "(GameplayTags=)" (Variables|Default|GetMetersTravelled))
    (Variables|Default|SetMetersTravelled 0)))
